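# Post-process AnimBP for the QuarterClaspZoom male head that damps head rotation when the character carries the RoundThumb_Gumbo cosmetic tag.
# The tag is checked on initialization and customization load.
# Each update with the tag and no emote, head's delta from ref pose relative to neck_02 is negated and scaled per axis, by default only yaw.
# An additive bone-space Modify Bone applies it, keeping the last value when updates stop.
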AnimBlueprint /BRCosmetics/Characters/Player/Male/Medium/Heads/M_MED_QuarterClaspZoom_Head/Meshes/M_MED_QuarterClaspZoom_Head_AnimBP_Postprocess
target_skeleton=/Game/Characters/Player/Common/Fortnite_Base_Head/Fortnite_Base_3L_Head_Skeleton
graphs [
 AnimGraph {nodes [
   n1 = "Transform (Modify) Bone - Bone: head"(
    ComponentPose=Local To Component(LocalPose=Input Pose())
    Alpha=1
    Translation="0,0,0"
    Rotation=Get head rot() -> (head rot:Rotator)
    Scale="1,1,1"
   ) {
    rotationMode=BMM_Additive
    rotationSpace=BCS_BoneSpace
    alphaBoolBlend {blendOption=Linear}
    boneToModify {boneName=head}
    rotation=<rot 50 50 50>
   }
   Output Pose(Result=Component To Local(ComponentPose=@n1))
  ]}
 EventGraph {nodes [
   Event BlueprintUpdateAnimation()
   n1 = Get bIsPlayingEmote()
   > Branch(Condition=AND Boolean(A=Get Is_PecanWear() -> (Is_PecanWear:bool), B=NOT Boolean(A=@n1)))
   n2 = GetOwnerSkeletalMeshComponent()
   n3 = BreakRotator(InRot=BreakTransform(InTransform=GetDeltaTransformFromRefPose(self=@n2, BoneName=head, BaseName=neck_02)).Rotation)
   n4 = Get Y percentage() -> (Y percentage:float)
   n5 = Get X percentage() -> (X percentage:float)
   n6 = Get Z percentage() -> (Z percentage:float)
   > Set head rot(head rot:Rotator=MakeRotator(
     Roll=float * float!(A=@n3, B=-1, C:double=@n5) -> (ReturnValue)
     Pitch=float * float(A=@n3.Pitch, B=-1, C=@n4)
     Yaw=float * float(A=@n3.Yaw, B=-1, C=@n6)
    )) -> (Output_Get:Rotator)
   n7 = Event OnCharacterCustomizationLoaded()
   Set Is_PecanWear(
    execute=[@n7.then, @n8.then]
    Is_PecanWear:bool=CharacterItemDefinitionHasAnyMetaTag(Tags=Make Array("[0]"=(TagName="Cosmetics.MetaTags.RoundThumb_Gumbo")))
   ) -> (Output_Get:bool)
   n8 = Event BlueprintInitializeAnimation()
   TryGetPawnOwner()
   TryGetPawnOwner()
  ]}
]
variables [
 As Fort Player Pawn Athena {type=FortPlayerPawnAthena}
 head rot {type=Rotator}
 Is_PecanWear {type=boolean}
 X percentage {type=number}
 Y percentage {type=number}
 Z percentage {type=number, default=0.55}
]
